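# Niagara system for the Tomb vehicle boost cosmetic, a blue-cyan exhaust effect offset from the car by the single-boost position.
# A socket-filtered ribbon draws the trail, while a multi-renderer sprite emitter uses spawn groups to vary size, velocity and lifetime for fire, smoke and glow layers.
# A burst emitter adds looping, pulsing mesh particles at the nozzle.
# A frontend-preview flag adjusts color, brightness and scale; painted-variant overrides are disabled.
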NiagaraSystem /VehicleCosmetics/Boosts/Boost_Tomb/FX/NS_Boost_Tomb
dependencies {
 usedDataInterfaces [
  &NiagaraDataInterfaceArrayFloat2
  &NiagaraDataInterfaceMeshRendererInfo
  &NiagaraDataInterfaceCurve
  &NiagaraDataInterfaceArrayFloat3
  &NiagaraDataInterfaceSpriteRendererInfo
  &NiagaraDataInterfaceColorCurve
  &NiagaraDataInterfaceArrayColor
  &NiagaraDataInterfaceSkeletalMesh
 ]
 usedDynamicInputs [&MakeFloatFromVector]
 usedModules [
  &self:SystemSpawnScript.SystemScriptSource.SystemScriptGraph.NiagaraNodeAssignment_0.SetVariables_#1
  &SystemState
 ]
}
emitters {
 NE_VC_SingleBoost_Ribbon_Parent {
  emitterUpdateScript [
   &EmitterState()
   &VC_FilteredSocketCount()
   &VC_RateQualityHandler(Epic Spawn Rate=45, Low Spawn Rate=30, Use Custom Rate?=true)
   &VC_SpawnRibbon_v1(SpawnRate=$Emitter.Spawn Rate)
  ]
  particleSpawnScript [
   &VC_PaintedVariantsOverride() {enabled=false}
   &VC_InitializeRibbon_v1(
    Color=&MakeCustomLinearColorFromBool(Bool=$User.bIsFrontendPreview, True Color=<rgba 0.01 0.0353 0.2 2>, False Color=<rgba 0.02 0.074 0.2 1.5>)
    Custom Offset=true
    Lifetime=&MakeCustomFloatFromBool(Bool=$User.bIsFrontendPreview, True Float=0.55, False Float=0.6)
    Local Vector Offset=&AddVectorToPosition(
     Position=<-80 0 0>
     Vector=&AddVectorToPosition(Position=$User.DTCarOffset, Vector=Linked() {linkedVariable {name=User.SingleBoostPos, type=NiagaraPosition}})
    )
    Offset Amount=1
    Ribbon Width=30
   )
   &AddVelocity(Velocity=&RandomRangeVector(Minimum=<-2250 0 40>, Maximum=<-2250 0 -40>)) {module=AddVelocity001}
   &ShapeLocation(MaskSpawnGroup=false, Override Seed=false, Sphere Radius=3, Sphere Surface Distribution=1, UseOwnerScale=true)
  ]
  particleUpdateScript [
   &ParticleState()
   &InheritSourceMovement(
    Applied Position Scale=&Multiply_VectorByFloat(
     Vector=<1 1 1>
     Float=&FloatFromCurve(FloatCurve=DataInterface() {propertyValues {Curve {keys [<at 0 1>,<at 1 0>]}}})
    )
    ApplyForce=false
    ApplyPosition=true
    ApplyPositionFalloff=false
   )
   &SolveForcesAndVelocity(Clamp Velocity=false, Limit Acceleration=false)
   &VC_UpdateRibbon_v1(Enable Skeletal Position=true)
   &ScaleRibbonWidth(Ribbon Width Scale=&FloatFromCurve(FloatCurve=DataInterface() {propertyValues {Curve {
        defaultValue=1
        keys [
         {interpMode=RCIM_Cubic, tangentMode=RCTM_User, time=0, value=0.5}
         {arriveTangent=0.0671, interpMode=RCIM_Cubic, leaveTangent=0.0671, tangentMode=RCTM_User, time=0.15, value=0.9833}
         {arriveTangent=-1.1868, interpMode=RCIM_Cubic, leaveTangent=-1.1868, tangentMode=RCTM_User, time=1, value=0}
        ]
       }}}))
   &ScaleColor(
    Scale Alpha=&FloatFromCurve(
     FloatCurve=DataInterface() {propertyValues {Curve {
        defaultValue=1
        keys [
         {arriveTangent=9.0725, interpMode=RCIM_Cubic, leaveTangent=9.0725, tangentMode=RCTM_User, time=0.1, value=0}
         {arriveTangent=-0.281, interpMode=RCIM_Cubic, leaveTangent=-0.281, tangentMode=RCTM_User, time=0.1408, value=1}
         {arriveTangent=-1.2059, interpMode=RCIM_Cubic, leaveTangent=-1.2059, time=0.5515, value=0.7}
         {arriveTangent=-0.0405, interpMode=RCIM_Cubic, leaveTangent=-0.0405, tangentMode=RCTM_User, time=0.97, value=0}
        ]
       }}}
     Scale Curve=3
    )
    ScaleA=true
    ScaleRGB=true
   )
  ]
  rendererClasses [&NiagaraRibbonRendererProperties]
  renderers [{rendererClass=&NiagaraRibbonRendererProperties}]
 }
 NE_VC_SingleBoost_Sprite_Parent {
  emitterUpdateScript [
   &EmitterState(Inactive Response=::NewEnumerator1, Life Cycle Mode=::NewEnumerator1, Loop Duration=1, UseLoopDelay=false)
   &SpawnRate(SpawnGroup=3, SpawnRate=10, Use Spawn Probability=false) {module=SpawnRate003}
   &SpawnRate(SpawnGroup=2, SpawnRate=50, Use Spawn Probability=false) {module=SpawnRate002}
   &SpawnRate(SpawnGroup=1, SpawnRate=60, Use Spawn Probability=false) {module=SpawnRate001}
   &SpawnRate(SpawnRate=400, Use Spawn Probability=false)
  ]
  particleSpawnScript [
   &VC_PaintedVariantsOverride() {enabled=false}
   &InitializeParticle(
    Color=<rgba 1.2 1.8574 2 0.5>
    Lifetime=/VehicleCosmetics/FX/Niagara/Modules/MergeEmitters/VC_FloatBySpawnGroup(
     SpawnGroup or ExecIndex:NiagaraInt32=0
     CustomInt:NiagaraInt32=0
     MaximumSpawnGroups:NiagaraInt32=4
     Float SG 0:NiagaraFloat=&RandomRangeFloat(Minimum=0.12, Maximum=0.25)
     Float SG 1:NiagaraFloat=0.025
     Float SG 2:NiagaraFloat=0.1
     Float SG 3:NiagaraFloat=0.01
     Float SG 4:NiagaraFloat=45
     Float SG 5:NiagaraFloat=0
    )
    Position Offset=&AddVectorToPosition(
     Position=&SelectVectorFromArray(
      Direct Array Index=&SpawnGroup()
      Vector Selection Array=DataInterface() {propertyValues {FloatData [<-65 0 0>, <-85 0 0>, <-120 0 0>, <-140 0 0>]}}
     )
     Vector=&AddVectorToPosition(Position=$User.DTCarOffset, Vector=Linked() {linkedVariable {name=User.SingleBoostPos, type=NiagaraPosition}})
    )
    Position Offset Coordinate Space=::Local
    Sprite Rotation Mode=::NewEnumerator2
    Sprite Size=&SelectVector2DFromArray(
     Direct Array Index=&SpawnGroup()
     Vector 2D Selection Array=DataInterface() {propertyValues {FloatData [<50 50>, <0 0>, <115 450>, <40 100>]}}
    )
    Sprite Size Mode=::NewEnumerator4
    UsePositionOffset=true
    Write Material Random=true
   )
   &AddVelocity(
    Velocity=&SelectVectorFromArray(
     Direct Array Index=&SpawnGroup()
     Vector Selection Array=DataInterface() {propertyValues {FloatData [<-1000 0 0>, <100 0 0>, <-500 0 0>, <-600 0 0>]}}
    )
    Velocity Speed Scale=&MakeCustomFloatFromBool(Bool=$User.bIsFrontendPreview, True Float=1, False Float=0.9)
   )
   /SetVariables(Particles.VisibilityTag=&SpawnGroup()) {isSetParameters=true, module=SetVariables_#2}
   &ShapeLocation(
    MaskSpawnGroup=false
    Override Seed=false
    Sphere Radius=&VC_FloatBySpawnGroup(
     SpawnGroup or ExecIndex:NiagaraInt32=0
     CustomInt:NiagaraInt32=0
     MaximumSpawnGroups:NiagaraInt32=2
     Float SG 0:NiagaraFloat=5
     Float SG 1:NiagaraFloat=2
     Float SG 2:NiagaraFloat=85
     Float SG 3:NiagaraFloat=35
     Float SG 4:NiagaraFloat=45
     Float SG 5:NiagaraFloat=0
    )
    UseOwnerScale=true
   )
  ]
  particleUpdateScript [
   &ParticleState()
   &SolveForcesAndVelocity(Clamp Velocity=false, Limit Acceleration=false)
   &SubUVAnimation(
    "Sprite Renderer "=DataInterface() {propertyValues {SpriteRenderer=&self:NE_VC_SingleBoost_Sprite_Parent_0.NiagaraSpriteRendererProperties_0}}
    UseEndFrame=false
    UseStartFrame=false
   )
   &SubUVAnimation(
    "Sprite Renderer "=DataInterface() {propertyValues {SpriteRenderer=&self:NE_VC_SingleBoost_Sprite_Parent_0.NiagaraSpriteRendererProperties_1}}
    Start Frame Range Override=20
    UseEndFrame=true
    UseStartFrame=true
   ) {module=SubUVAnimation001}
   &SubUVAnimation(
    "Sprite Renderer "=DataInterface() {propertyValues {SpriteRenderer=&self:NE_VC_SingleBoost_Sprite_Parent_0.NiagaraSpriteRendererProperties_2}}
    Start Frame Range Override=23
    SubUV Animation Mode:ENiagara_SubUVLookupModeV2=::NewEnumerator2
    UseEndFrame=true
    UseStartFrame=true
   ) {module=SubUVAnimation002}
   /SetVariables(
    Particles.FireColor:LinearColor=&MakeCustomLinearColorFromBool(
     Bool=$User.bIsFrontendPreview
     True Color=&ColorFromCurve(ColorCurve=DataInterface() {propertyValues {
        AlphaCurve {keys [<at 0 1 RCIM_Cubic>,<at 1 0 RCIM_Cubic>]}
        BlueCurve {keys [<at 0.3902 4 RCIM_Cubic>]}
        GreenCurve {keys [<at 0.3902 3.5722 RCIM_Cubic>]}
        RedCurve {keys [<at 0.3902 1.6 RCIM_Cubic>]}
       }})
     False Color=&ColorFromCurve(ColorCurve=DataInterface() {propertyValues {
        AlphaCurve {keys [<at 0 20 RCIM_Cubic>,<at 1 0 RCIM_Cubic>]}
        BlueCurve {keys [<at 0.3902 2 RCIM_Cubic>,<at 0.4156 0.8 RCIM_Cubic>]}
        GreenCurve {keys [<at 0.3902 1.7861 RCIM_Cubic>,<at 0.4156 1.1 RCIM_Cubic>]}
        RedCurve {keys [<at 0.3902 0.8 RCIM_Cubic>,<at 0.4156 2 RCIM_Cubic>]}
       }})
    )
    Particles.FireOpaqueColor:LinearColor=&MultiplyLinearColor(
     Linear Color=<rgba 1 0.6 0.5 1>
     Scale Factor=&MakeVector4FromVectorAndFloat(
      "Vector (RGB)"=<1 1 1>
      "Float (Alpha)"=&FloatFromCurve(FloatCurve=DataInterface() {propertyValues {Curve {
          defaultValue=1
          keys [
           {arriveTangent=9.0725, interpMode=RCIM_Cubic, leaveTangent=9.0725, tangentMode=RCTM_User, time=0.5052, value=0}
           {arriveTangent=0.2164, interpMode=RCIM_Cubic, leaveTangent=0.2164, tangentMode=RCTM_User, time=0.5775, value=0.35}
           {arriveTangent=-0.0822, interpMode=RCIM_Cubic, leaveTangent=-0.0822, tangentMode=RCTM_User, time=0.7464, value=0.35}
           {arriveTangent=-0.0405, interpMode=RCIM_Cubic, leaveTangent=-0.0405, tangentMode=RCTM_User, time=1, value=0}
          ]
         }}})
     )
    )
    Particles.FireOpaqueSize:Vector2f=&Vector2DFromFloat(Value=&FloatFromCurve(
      FloatCurve=DataInterface() {propertyValues {Curve {
         defaultValue=0
         keys [
          {arriveTangent=3.8522, interpMode=RCIM_Cubic, leaveTangent=3.8522, tangentMode=RCTM_User, time=0.0355, value=0.4}
          {arriveTangent=0.6914, interpMode=RCIM_Cubic, leaveTangent=0.6914, tangentMode=RCTM_User, time=0.4827, value=1.3217}
          {arriveTangent=-0.0864, interpMode=RCIM_Cubic, leaveTangent=-0.0864, tangentMode=RCTM_User, time=1, value=1.6878}
         ]
        }}}
      Scale Curve=55
     ))
    Particles.GlowColor=&MultiplyLinearColor(
     Linear Color=<rgba 2.8609 4.9442 5 0.1739>
     Scale Factor=&MakeVector4FromVectorAndFloat("Vector (RGB)"=<1 1 1>, "Float (Alpha)"=0.3)
    )
    Particles.SmokeColor:LinearColor=&SelectLinearColorFromArray(
     Array Sampling Mode=::NewEnumerator0
     Direct Array Index=?
     Color Selection Array=DataInterface() {propertyValues {ColorData [<rgba 5 1.7917 1.5 1>, <rgba 0 3.7293 5 0.4435>]}}
     Random Seed=0
    )
    Particles.SmokeRotation:NiagaraFloat=&RandomRangeFloat(Minimum=0, Maximum=0)
    Particles.SpriteSize Fire:Vector2f=&Vector2DFromFloat(Value=&FloatFromCurve(
      FloatCurve=DataInterface() {propertyValues {Curve {
         defaultValue=0
         keys [
          {arriveTangent=2.28, interpMode=RCIM_Cubic, leaveTangent=2.28, tangentMode=RCTM_User, time=0, value=0.4}
          {arriveTangent=0.0653, interpMode=RCIM_Cubic, leaveTangent=0.0653, tangentMode=RCTM_User, time=0.2, value=1.25}
          {interpMode=RCIM_Cubic, tangentMode=RCTM_User, time=0.6, value=1.25}
          {arriveTangent=-2.8627, interpMode=RCIM_Cubic, leaveTangent=-2.8627, tangentMode=RCTM_User, time=1, value=0.7864}
         ]
        }}}
      Scale Curve=55
     ))
   ) {isSetParameters=true}
   &ScaleColor(
    Scale Alpha=&FloatFromCurve(FloatCurve=DataInterface() {propertyValues {Curve {keys [<at 0.4 1 RCIM_Cubic>,<at 1 0 RCIM_Cubic>]}}})
    Scale RGB=<0 0 0>
    ScaleA=true
    ScaleRGB=true
   )
  ]
  rendererClasses [&NiagaraSpriteRendererProperties]
  renderers [
   {rendererClass=&NiagaraSpriteRendererProperties}
   {rendererClass=&NiagaraSpriteRendererProperties, rendererIndex=1}
   {rendererClass=&NiagaraSpriteRendererProperties, rendererIndex=2}
   {rendererClass=&NiagaraSpriteRendererProperties, rendererIndex=3}
  ]
 }
 NE_VC_SingleBoost_Sprite_Parent002 {
  emitterUpdateScript [
   &EmitterState(
    Inactive Response=::NewEnumerator1
    Life Cycle Mode=::NewEnumerator1
    Loop Behavior=::NewEnumerator1
    Loop Duration Mode=::NewEnumerator1
    UseLoopDelay=false
   )
   &SpawnBurst_Instantaneous(Spawn Count=10, Spawn Group=1, Use Spawn Probability=false, UseLoopCountLimit=false)
   &SpawnRate(SpawnRate=200, Use Spawn Probability=false) {enabled=false}
  ]
  particleSpawnScript [
   &VC_PaintedVariantsOverride() {enabled=false}
   &InitializeParticle(
    Color=<rgba 0.3 1.697 2 0.5>
    Lifetime=0.15
    Mesh Renderer Array Visibility Mode=::NewEnumerator3
    Mesh Renderer Info=DataInterface() {propertyValues {MeshRenderer=&self:NE_VC_SingleBoost_Sprite_Parent002_1.NiagaraMeshRendererProperties_0}}
    Mesh Scale Max=&MakeCustomVectorFromBool(Bool=$User.bIsFrontendPreview, True Vector=<15 9 9>, False Vector=<7 9 9>)
    Mesh Scale Min=&MakeCustomVectorFromBool(Bool=$User.bIsFrontendPreview, True Vector=<12 9 9>, False Vector=<5 9 9>)
    Mesh Scale Mode=::NewEnumerator2
    Position Offset=&AddVectorToPosition(
     Position=<-50 0 0>
     Vector=&AddVectorToPosition(Position=$User.DTCarOffset, Vector=Linked() {linkedVariable {name=User.SingleBoostPos, type=NiagaraPosition}})
    )
    Position Offset Coordinate Space=::Local
    Sprite Rotation Mode=::NewEnumerator2
    UsePositionOffset=true
    Write Material Random=true
   )
   &AddVelocity(Velocity=<0 0 0>)
   /SetVariables(Particles.VisibilityTag=&SpawnGroup()) {isSetParameters=true, module=SetVariables_#2}
  ]
  particleUpdateScript [
   &ParticleState(Kill Particles When Lifetime Has Elapsed=false, Loop Particles Lifetime=true)
   &ScaleMeshSize(Scale Factor=&VectorFromFloat(Value=&FloatFromCurve(FloatCurve=DataInterface() {propertyValues {Curve {keys [<at 0 1>,<at 1 0.65>]}}})))
   &SolveForcesAndVelocity(Clamp Velocity=false, Limit Acceleration=false)
   &ScaleColor(
    Scale Alpha=&FloatFromCurve(
     FloatCurve=DataInterface() {propertyValues {Curve {keys [<at 0.25 1 RCIM_Cubic>,<at 1 0.5 RCIM_Cubic>]}}}
     Scale Curve=&MakeCustomFloatFromBool(Bool=$User.bIsFrontendPreview, True Float=1.5, False Float=0.5)
    )
    ScaleA=true
    ScaleRGB=true
   )
  ]
  rendererClasses [&NiagaraSpriteRendererProperties, &NiagaraMeshRendererProperties]
  renderers [
   {rendererClass=&NiagaraSpriteRendererProperties}
   {rendererClass=&NiagaraMeshRendererProperties, rendererIndex=1}
  ]
 }
}
summary {
 systemName=NS_Boost_Tomb
 emitters [
  {
   emitterName=NE_VC_SingleBoost_Sprite_Parent002
   rendererClasses [&NiagaraSpriteRendererProperties, &NiagaraMeshRendererProperties]
  }
  {emitterName=NE_VC_SingleBoost_Ribbon_Parent, rendererClasses [&NiagaraRibbonRendererProperties]}
  {emitterName=NE_VC_SingleBoost_Sprite_Parent, rendererClasses [&NiagaraSpriteRendererProperties]}
 ]
 userVariables [
  User.bIsFrontendPreview {type=NiagaraBool}
  User.bIsPainted {type=NiagaraBool}
  User.ColorVariantID {type=NiagaraInt32}
  User.DTCarOffset {type=NiagaraPosition, description="", defaultValue {}}
  User.PaintedColor {type=LinearColor, description="", defaultValue {struct=LinearColor, value=<rgba 1 1 1 1>}}
  User.PaintedScalar {type=NiagaraFloat, defaultValue=1, description=""}
  User.PortCount {type=NiagaraInt32, defaultValue=2}
  User.SingleBoostPos {type=NiagaraPosition, description="", defaultValue {}}
 ]
}
